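# Sound cue for a water Sprite's grab sound effect, using Sprite foley attenuation and the creature foley sound class.
# A Modulator varies playback over a quality-level switch between two equally weighted Random nodes.
# The first two quality slots draw from all eight grab waves, and the last two from only the four odd-numbered ones.
# An owner-limited concurrency override caps simultaneous plays, stopping the oldest.
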SoundCue /Sprites_WhisperRemedySweetSprite/Assets/Sounds/Cues/SFX/SC_Sprite_Water_Grab
root=Modulator_0
nodes [
 {id=Modulator_0, children [QualityLevel_0], properties {volumeMin=0.8}}
 {id=QualityLevel_0, children [Random_0,Random_0,Random_1,Random_1]}
 {
  id=Random_0
  children [WavePlayer_0,WavePlayer_1,WavePlayer_2,WavePlayer_3,WavePlayer_4,WavePlayer_5,WavePlayer_6,WavePlayer_7]
  properties {bShouldExcludeFromBranchCulling=true, weights [1,1,1,1,1,1,1,1]}
 }
 {id=Random_1, children [WavePlayer_8, WavePlayer_9, WavePlayer_10, WavePlayer_11], properties {weights [1,1,1,1]}}
 {id=WavePlayer_0, properties {soundWaveAssetPtr=/Sprites_WhisperRemedySweetSprite/Assets/Sounds/Waves/SFX/Sprite_Water_Grab_01}}
 {id=WavePlayer_1, properties {soundWaveAssetPtr=/Sprites_WhisperRemedySweetSprite/Assets/Sounds/Waves/SFX/Sprite_Water_Grab_02}}
 {id=WavePlayer_10, properties {soundWaveAssetPtr=/Sprites_WhisperRemedySweetSprite/Assets/Sounds/Waves/SFX/Sprite_Water_Grab_05}}
 {id=WavePlayer_11, properties {soundWaveAssetPtr=/Sprites_WhisperRemedySweetSprite/Assets/Sounds/Waves/SFX/Sprite_Water_Grab_07}}
 {id=WavePlayer_2, properties {soundWaveAssetPtr=/Sprites_WhisperRemedySweetSprite/Assets/Sounds/Waves/SFX/Sprite_Water_Grab_03}}
 {id=WavePlayer_3, properties {soundWaveAssetPtr=/Sprites_WhisperRemedySweetSprite/Assets/Sounds/Waves/SFX/Sprite_Water_Grab_04}}
 {id=WavePlayer_4, properties {soundWaveAssetPtr=&Sprite_Water_Grab_05}}
 {id=WavePlayer_5, properties {soundWaveAssetPtr=/Sprites_WhisperRemedySweetSprite/Assets/Sounds/Waves/SFX/Sprite_Water_Grab_06}}
 {id=WavePlayer_6, properties {soundWaveAssetPtr=&Sprite_Water_Grab_07}}
 {id=WavePlayer_7, properties {soundWaveAssetPtr=/Sprites_WhisperRemedySweetSprite/Assets/Sounds/Waves/SFX/Sprite_Water_Grab_08}}
 {id=WavePlayer_8, properties {soundWaveAssetPtr=&Sprite_Water_Grab_01}}
 {id=WavePlayer_9, properties {soundWaveAssetPtr=&Sprite_Water_Grab_03}}
]
properties {
 attenuationSettings=/Game/Sounds/Fort_Foley_Sounds/FoleyLibraries/Settings/Foley_Sprites_Attenuation_Basic_Att
 bOverrideConcurrency=true
 maxDistance=1536
 soundClassObject=/Game/Sounds/FortSoundClassesAndModes/Fort_Creatures_Foley
 volumeMultiplier=1
 concurrencyOverrides {bLimitToOwner=true, maxCount=2, resolutionRule=StopOldest, volumeScaleAttackTime=0.05}
}
