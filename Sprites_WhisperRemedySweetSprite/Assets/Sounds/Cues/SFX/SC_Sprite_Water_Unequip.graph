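# Sound cue for the water Sprite's unequip sound effect, attenuated with basic Sprite foley settings in the creature foley sound class.
# A quality switch sends its first two slots to a random pick of eight waves and its last two to a subset of four.
# A modulator above the switch varies volume per play.
# Concurrency is limited per owner to two instances, stopping the oldest.
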SoundCue /Sprites_WhisperRemedySweetSprite/Assets/Sounds/Cues/SFX/SC_Sprite_Water_Unequip
root=Modulator_0
nodes [
 {id=Modulator_0, children [QualityLevel_0], properties {volumeMin=0.8}}
 {id=QualityLevel_0, children [Random_2,Random_2,Random_3,Random_3]}
 {
  id=Random_2
  children [WavePlayer_12, WavePlayer_13, WavePlayer_14, WavePlayer_15, WavePlayer_16, WavePlayer_17, WavePlayer_18, WavePlayer_19]
  properties {bShouldExcludeFromBranchCulling=true, weights [1,1,1,1,1,1,1,1]}
 }
 {id=Random_3, children [WavePlayer_20, WavePlayer_21, WavePlayer_22, WavePlayer_23], properties {weights [1,1,1,1]}}
 {id=WavePlayer_12, properties {soundWaveAssetPtr=/Sprites_WhisperRemedySweetSprite/Assets/Sounds/Waves/SFX/Sprite_Water_Unequip_01}}
 {id=WavePlayer_13, properties {soundWaveAssetPtr=/Sprites_WhisperRemedySweetSprite/Assets/Sounds/Waves/SFX/Sprite_Water_Unequip_02}}
 {id=WavePlayer_14, properties {soundWaveAssetPtr=/Sprites_WhisperRemedySweetSprite/Assets/Sounds/Waves/SFX/Sprite_Water_Unequip_03}}
 {id=WavePlayer_15, properties {soundWaveAssetPtr=/Sprites_WhisperRemedySweetSprite/Assets/Sounds/Waves/SFX/Sprite_Water_Unequip_04}}
 {id=WavePlayer_16, properties {soundWaveAssetPtr=/Sprites_WhisperRemedySweetSprite/Assets/Sounds/Waves/SFX/Sprite_Water_Unequip_05}}
 {id=WavePlayer_17, properties {soundWaveAssetPtr=/Sprites_WhisperRemedySweetSprite/Assets/Sounds/Waves/SFX/Sprite_Water_Unequip_06}}
 {id=WavePlayer_18, properties {soundWaveAssetPtr=/Sprites_WhisperRemedySweetSprite/Assets/Sounds/Waves/SFX/Sprite_Water_Unequip_07}}
 {id=WavePlayer_19, properties {soundWaveAssetPtr=/Sprites_WhisperRemedySweetSprite/Assets/Sounds/Waves/SFX/Sprite_Water_Unequip_08}}
 {id=WavePlayer_20, properties {soundWaveAssetPtr=&Sprite_Water_Unequip_01}}
 {id=WavePlayer_21, properties {soundWaveAssetPtr=&Sprite_Water_Unequip_03}}
 {id=WavePlayer_22, properties {soundWaveAssetPtr=&Sprite_Water_Unequip_05}}
 {id=WavePlayer_23, properties {soundWaveAssetPtr=&Sprite_Water_Unequip_07}}
]
properties {
 attenuationSettings=/Game/Sounds/Fort_Foley_Sounds/FoleyLibraries/Settings/Foley_Sprites_Attenuation_Basic_Att
 bOverrideConcurrency=true
 maxDistance=1536
 soundClassObject=/Game/Sounds/FortSoundClassesAndModes/Fort_Creatures_Foley
 volumeMultiplier=0.8
 concurrencyOverrides {bLimitToOwner=true, maxCount=2, resolutionRule=StopOldest, volumeScaleAttackTime=0.05}
}
